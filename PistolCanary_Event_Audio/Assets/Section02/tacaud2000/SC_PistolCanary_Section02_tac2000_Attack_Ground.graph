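# Sound cue for the ground impact of a tentacle attack in Section 02 of the PistolCanary event.
# A single wave player plays the tentacle ground-impact wave with no randomization, mixing or modulation.
# It uses the beam-ground attenuation settings with a long max distance, routes to the Fort_Events sound class, and plays at reduced volume.
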SoundCue /PistolCanary_Event_Audio/Assets/Section02/tacaud2000/SC_PistolCanary_Section02_tac2000_Attack_Ground
root=WavePlayer_0
nodes [{id=WavePlayer_0, properties {soundWaveAssetPtr=/PistolCanary_Event_Audio/Assets/Section02/wav/PistolCanary_Tentacle_Attk_2000_Ground_Impact}}]
properties {
 attenuationSettings=/PistolCanary_Audio/Assets/Settings/ATT_PistolCanary_Beam_Ground
 maxDistance=22000
 soundClassObject=&Fort_Events
 volumeMultiplier=0.55
}
